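; Tutorial sample Blueprint class from the engine's Blueprint tutorials, demonstrating basic event graph and Construction Script logic.
; BeginPlay and Tick each lead through a Gate into an authority check, which stores the current Game Mode in a variable when it passes a not-equal object test.
; The Construction Script sets SampleVariable to true.
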(event EventTick
  (params (DeltaSeconds :type Float%20%28single-precision%29))
  (bind _returnvalue (Game|GetGameMode))
  (Utilities|FlowControl|Gate true
    (:Exit
      (Networking|SwitchHasAuthority
        (:Authority
          (if (Utilities|NotEqual(Object) _returnvalue 0)
            (Variables|Default|SetStoredGameMode _returnvalue)))))))

(event EventBeginPlay
  (bind _returnvalue (Game|GetGameMode))
  (Utilities|FlowControl|Gate true
    (:Exit
      (Networking|SwitchHasAuthority
        (:Authority
          (if (Utilities|NotEqual(Object) _returnvalue 0)
            (Variables|Default|SetStoredGameMode _returnvalue)))))))

(fn ConstructionScript
  (Variables|Default|SetSampleVariable true))
