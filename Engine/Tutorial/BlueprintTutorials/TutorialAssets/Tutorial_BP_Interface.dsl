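; Blueprint Interface asset from the engine's Blueprint tutorials, serving as a sample interface.
; It declares a single function, TutorialSampleFunction, with no inputs or outputs shown and no logic.
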(fn TutorialSampleFunction)
